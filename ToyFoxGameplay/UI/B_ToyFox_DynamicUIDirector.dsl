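; Dynamic UI director for the ToyFox mode that toggles player HUD, alert and enemy-player scenes for the local player.
; View-target changes resolve the viewed pawn, via bike instigator with timed retry if needed, then tag watchers on it toggle common, primary and secondary HUDs and hide elements.
; Game-state event changes drive alert and enemy scenes, the latter gated by giant's team.
; Delta gates skip redundant scene changes; secondary HUD also adjusts touch region division.
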(event EventBeginPlay
  (bind _returnvalue (Fortnite|Athena|GameState|GetGameStateAthena))
  ;[0.000659 0.09375 0.01458 0.5] BeginPlay: Bind to know when the view target has changed.
  (|CallParentFunction)
  (Utilities|FlowControl|Sequence
    (:then_0
      (bind _asfort_player_controller (Utilities|Casting|CastToFortPlayerController (DynamicUIDirector|GetOwningLocalPlayerController)))
      (Fortnite|FortShippingLog self (Utilities|String|Append "BeginPlay: Successfully binding to view target changed for controller " (Fortnite|GetPlayerNameSafe _asfort_player_controller)))
      (EventDispatchers|BindEventtoOnViewTargetChanged _asfort_player_controller (EventDispatchers|CreateEvent :!event Custom|HandleLocalPlayerViewTargetChanged))
      (CallFunction|HandleLocalPlayerViewTargetChanged _asfort_player_controller 0 (Pawn|GetViewTarget _asfort_player_controller)))
    (:then_1
      (Utilities|IsValid _returnvalue
        (:Is%20Valid
          (bind _returnvalue_2 (Actor|GetComponentbyClass _returnvalue "/ToyFoxGameplay/Gameplay/B_ToyFoxGameStateComponent.B_ToyFoxGameStateComponent_C"))
          (bind _output_get (Variables|Default|SetToyFoxGameStateComponent _returnvalue_2))
          (Utilities|IsValid _output_get
            (:Is%20Valid
              (Default|BindEventtoOnEventStateChanged _output_get (EventDispatchers|CreateEvent :!event Custom|HandleEventStateChanged)))))))))

(event Custom|HandleLocalPlayerViewTargetChanged
  (params (PlayerController :type Fort%20Player%20Controller%20Object%20Reference) (OldViewTarget :type Actor%20Object%20Reference) (NewViewTarget :type Actor%20Object%20Reference))
  (bind _returnvalue (Ability|GetFortAbilitySystemComponent (Variables|Transient|GetCurrentViewPawn)))
  (bind _cachedviewtarget (Variables|Default|GetCachedViewTarget))
  (bind _viewtargetasyncactions (Variables|Transient|GetViewTargetAsyncActions))
  (bind _currentviewpawn (Variables|Transient|GetCurrentViewPawn))
  ;[0 0.01961 0.15 0.5] HandleLocalPlayerViewTargetChanged: Handle view target changed and fetch a valid ASC
  (Variables|Default|SetCachedViewTarget NewViewTarget)
  (Utilities|Time|ClearandInvalidateTimerbyHandle (Variables|Default|GetRetryInstigatorTimerHandle))
  (Fortnite|FortShippingLog self (Utilities|String|Append "HandleLocalPlayerViewTargetChanged: View target for player controller " (Fortnite|GetPlayerNameSafe PlayerController) " changed from " (Utilities|GetDisplayName OldViewTarget) " to " (Utilities|GetDisplayName NewViewTarget)))
  (Utilities|FlowControl|Sequence
    (:then_0
      (Ability|Async|EndAction _viewtargetasyncactions)
      (Utilities|Array|Clear _viewtargetasyncactions)
      ;;[0.05411 0.06992 0.15 0.5] Disable all the HUD scenes so they can be reinitialized based on the new view target.
      (CallFunction|SetCommonPlayerHUD false)
      (CallFunction|SetPrimaryPlayerHUD false)
      (bind _currentviewpawn_2 (Variables|Transient|GetCurrentViewPawn)
        (:then
          (EventDispatchers|UnbindEventfromOnAbilitySystemComponentInitialized _currentviewpawn_2 (EventDispatchers|CreateEvent :!event Custom|OnASCInitialized) :!class /Script/FortniteGame.FortPlayerPawn)
          (Fortnite|FortShippingLog self (Utilities|String|Append "HandleLocalPlayerViewTargetChanged: Unbinding from ability system component initialized for current view pawn: " (Utilities|GetDisplayName _currentviewpawn_2))))
        (:else)))
    (:then_1
      (bind _asfort_player_pawn (Utilities|Casting|CastToFortPlayerPawn _cachedviewtarget)
        (:then
          (Variables|Transient|SetCurrentViewPawn _asfort_player_pawn))
        (:CastFailed
          (bind _asfort_bike_kid (Utilities|Casting|CastToFortBikeKid _cachedviewtarget)
            (:then
              (bind _asfort_player_pawn_2 (Utilities|Casting|CastToFortPlayerPawn (Game|GetInstigator _asfort_bike_kid))
                (:then
                  (Variables|Transient|SetCurrentViewPawn _asfort_player_pawn_2))
                (:CastFailed
                  (bind _returnvalue_2 (Utilities|Time|SetTimerbyEvent (EventDispatchers|CreateEvent :!event Custom|RetryInstigator) 0.2))
                  (Variables|Default|SetRetryInstigatorTimerHandle _returnvalue_2))))
            (:CastFailed
              (CallFunction|SetSecondaryPlayerHUD false)))))
      (Utilities|IsValid _returnvalue
        (:Is%20Valid
          (CallFunction|OnASCInitialized _returnvalue _currentviewpawn))
        (:Is%20Not%20Valid
          (Fortnite|FortShippingLog self (Utilities|String|Append "HandleLocalPlayerViewTargetChanged: CurrentViewPawn " (Utilities|GetDisplayName _currentviewpawn) " has no ability system component. "))
          (CallFunction|SetSecondaryPlayerHUD false)))
      (EventDispatchers|BindEventtoOnAbilitySystemComponentInitialized _currentviewpawn (EventDispatchers|CreateEvent :!event Custom|OnASCInitialized) :!class /Script/FortniteGame.FortPlayerPawn) ; ASC might be uninitialized or could change.
    )))

(event Custom|OnASCInitialized
  (params (AbilitySystemComponent :type Fort%20Ability%20System%20Component%20Object%20Reference) (AffectedPawn :type Fort%20Player%20Pawn%20Object%20Reference))
  ;[0 0.01961 0.15 0.5] HandleLocalPlayerViewTargetChanged: Handle view target changed and fetch a valid ASC
  (Fortnite|FortShippingLog self (Utilities|String|Append "OnASCInitialized: AbilitySystemComponent: " (Utilities|GetDisplayName AbilitySystemComponent) " AffectedPawn: " (Utilities|GetDisplayName AffectedPawn)))
  (CallFunction|RegisterTagEvents))

(event Custom|RetryInstigator
  (bind _returnvalue (Ability|GetFortAbilitySystemComponent (Variables|Transient|GetCurrentViewPawn)))
  (bind _currentviewpawn (Variables|Transient|GetCurrentViewPawn))
  ;[0 0.01961 0.15 0.5] HandleLocalPlayerViewTargetChanged: Handle view target changed and fetch a valid ASC
  (bind _asfort_bike_kid (Utilities|Casting|CastToFortBikeKid (Variables|Default|GetCachedViewTarget))
    (:then
      (bind _asfort_player_pawn (Utilities|Casting|CastToFortPlayerPawn (Game|GetInstigator _asfort_bike_kid))
        (:then
          (Variables|Transient|SetCurrentViewPawn _asfort_player_pawn)
          (Utilities|IsValid _returnvalue
            (:Is%20Valid
              (CallFunction|OnASCInitialized _returnvalue _currentviewpawn))
            (:Is%20Not%20Valid
              (Fortnite|FortShippingLog self (Utilities|String|Append "HandleLocalPlayerViewTargetChanged: CurrentViewPawn " (Utilities|GetDisplayName _currentviewpawn) " has no ability system component. "))
              (CallFunction|SetSecondaryPlayerHUD false)))
          (EventDispatchers|BindEventtoOnAbilitySystemComponentInitialized _currentviewpawn (EventDispatchers|CreateEvent :!event Custom|OnASCInitialized) :!class /Script/FortniteGame.FortPlayerPawn) ; ASC might be uninitialized or could change.
        )
        (:CastFailed
          (bind _returnvalue_2 (Utilities|Time|SetTimerbyEvent (EventDispatchers|CreateEvent :!event Custom|RetryInstigator) 0.2))
          (Variables|Default|SetRetryInstigatorTimerHandle _returnvalue_2))))
    (:CastFailed
      (CallFunction|SetSecondaryPlayerHUD false))))

(event Custom|RegisterTagEvents
  (bind _hideallhudelementstagquery (Variables|Settings|GetHideAllHUDElementsTagQuery))
  (bind _currentviewpawn (Variables|Transient|GetCurrentViewPawn))
  (bind _commonplayergameplaytagquery (Variables|Settings|GetCommonPlayerGameplayTagQuery))
  (bind _secondarypilottagquery (Variables|Settings|GetSecondaryPilotTagQuery))
  ;[0.1359 0.02769 0.15 0.5] Register Tag Events on the pawn ASC
  (bind _asynctaskproxy (Ability|Tasks|WaitGameplayTagQueryOnActor _currentviewpawn _commonplayergameplaytagquery)
    (:then
      (Utilities|Array|Add (Variables|Transient|GetViewTargetAsyncActions) _asynctaskproxy)
      (bind _asynctaskproxy_2 (Ability|Tasks|WaitGameplayTagQueryOnActor _currentviewpawn _commonplayergameplaytagquery "WhenFalse")
        (:then
          (Utilities|Array|Add (Variables|Transient|GetViewTargetAsyncActions) _asynctaskproxy_2)
          (bind (_asynctaskproxy_3 _tagcount) (Ability|Tasks|WaitGameplayTagCountChangedOnActor _currentviewpawn "(TagName=\"ToyFox.Pawn\")")
            (:then
              (Utilities|Array|Add (Variables|Transient|GetViewTargetAsyncActions) _asynctaskproxy_3)
              (CallFunction|SetPrimaryPlayerHUD (GameplayTags|HasMatchingGameplayTag _currentviewpawn "(TagName=\"ToyFox.Pawn\")")) ; Needs to be checked explicitly since the async action doesn't check for this right away.
              (bind _asynctaskproxy_4 (Ability|Tasks|WaitGameplayTagQueryOnActor _currentviewpawn _secondarypilottagquery)
                (:then
                  (Utilities|Array|Add (Variables|Transient|GetViewTargetAsyncActions) _asynctaskproxy_4)
                  (bind _asynctaskproxy_5 (Ability|Tasks|WaitGameplayTagQueryOnActor _currentviewpawn _secondarypilottagquery "WhenFalse")
                    (:then
                      (Utilities|Array|Add (Variables|Transient|GetViewTargetAsyncActions) _asynctaskproxy_5)
                      (bind _asynctaskproxy_6 (Ability|Tasks|WaitGameplayTagQueryOnActor _currentviewpawn _hideallhudelementstagquery)
                        (:then
                          (Utilities|Array|Add (Variables|Transient|GetViewTargetAsyncActions) _asynctaskproxy_6)
                          (bind _asynctaskproxy_7 (Ability|Tasks|WaitGameplayTagQueryOnActor _currentviewpawn _hideallhudelementstagquery "WhenFalse")
                            (:then
                              (Utilities|Array|Add (Variables|Transient|GetViewTargetAsyncActions) _asynctaskproxy_7))
                            (:Triggered
                              (CallFunction|SetHideAllHUDElements false))))
                        (:Triggered
                          (CallFunction|SetHideAllHUDElements true))))
                    (:Triggered
                      (CallFunction|SetSecondaryPlayerHUD false))))
                (:Triggered
                  (CallFunction|SetSecondaryPlayerHUD true))))
            (:TagCountChanged
              (CallFunction|SetPrimaryPlayerHUD (> _tagcount 0)))))
        (:Triggered
          (CallFunction|SetCommonPlayerHUD false))))
    (:Triggered
      (CallFunction|SetCommonPlayerHUD true))))

(event Custom|SetCommonPlayerHUD
  (params (bIsInside :type Boolean))
  ;[0.08226 0.07338 0.15 0.5] SetHUDState: Sets the enable/disable for the common toyfox HUD. Run as true for all pilots.
  (CallFunction|SetHideHUDElements bIsInside (Variables|Settings|GetCommonPlayerHUDElementsToHide))
  (CallFunction|SetEnableSceneCommonHUD bIsInside))

(event Custom|SetPrimaryPlayerHUD
  (params (bEnable :type Boolean))
  ;[0.1084 0.04014 0.15 0.5] Sets primary player HUD state.
  (CallFunction|SetEnableScenePrimaryPlayerHUD bEnable))

(event Custom|SetSecondaryPlayerHUD
  (params (bEnable :type Boolean))
  ;[0.004616 0.007258 0.15 0.5] Sets secondary player HUD state.
  (CallFunction|SetEnableSceneSecondaryPlayerHUD bEnable)
  (CallFunction|SetHideHUDElements bEnable (Variables|Settings|GetSecondaryPlayerHUDElementsToHide)))

(event EventEndPlay
  (params (EndPlayReason :type EEndPlayReason%20Enum :default "Destroyed"))
  (bind _returnvalue (Fortnite|Athena|GameState|GetGameStateAthena))
  ;[0.08333 0.0447 0 0.5] EndPlay: Clean up delegates.
  (Utilities|FlowControl|Sequence
    (:then_0
      (Utilities|Time|ClearandInvalidateTimerbyHandle (Variables|Default|GetRetryInstigatorTimerHandle))
      (Utilities|IsValid _returnvalue
        (:Is%20Valid
          (bind _returnvalue_2 (Actor|GetComponentbyClass _returnvalue "/ToyFoxGameplay/Gameplay/B_ToyFoxGameStateComponent.B_ToyFoxGameStateComponent_C"))
          (Utilities|IsValid _returnvalue_2
            (:Is%20Valid
              (Default|UnbindEventfromOnEventStateChanged _returnvalue_2 (EventDispatchers|CreateEvent :!event Custom|HandleEventStateChanged)))))))
    (:then_1
      (|CallParentFunction EndPlayReason))))

(event Custom|HandleEventStateChanged
  (params (EventState :type ToyFoxEventStateEnum%20Enum :default "NewEnumerator0"))
  ;[1 1 1] HandleEventStateChanged: Handle event state changed.
  (Utilities|FlowControl|Sequence
    (:then_0
      (switch Utilities|FlowControl|Switch|SwitchonToyFoxEventStateEnum EventState
        (:NewEnumerator0
          _)
        (:NewEnumerator5
          (CallFunction|SetEnableSceneAlert true :!label _label))
        (:NewEnumerator1
          (goto _label))
        (:NewEnumerator2
          (goto _label))
        (:NewEnumerator3
          (goto _label))
        (:NewEnumerator4
          (Utilities|FlowControl|Delay 25.0) ; Give time for any remaining UI alerts to end.
        ))
      (CallFunction|SetEnableSceneAlert false))
    (:then_1
      (CallFunction|SetEnableSceneEnemyPlayers (Utilities|Select :Index EventState :NewEnumerator0 false :NewEnumerator5 false :NewEnumerator1 false :NewEnumerator2 false :NewEnumerator3 true :NewEnumerator4 false)))))

(fn IsActorAthenaGameParticipant
  (params (Actor :type Actor%20Object%20Reference))
  (bind _returnvalue (Fortnite|GetFortPlayerControllerfromActor Actor))
  (Utilities|FlowControl|Sequence
    (:then_0
      (Utilities|IsValid _returnvalue
        (:Is%20Valid
          (bind _asfort_player_state_athena (Utilities|Casting|CastToFortPlayerStateAthena (Variables|/Script/Engine.Controller|GetPlayerState _returnvalue)))
          (if (Class|FortPlayerStateAthena|IsAnAthenaGameParticipant _asfort_player_state_athena)
            (return (Class|FortPlayerStateAthena|IsAnAthenaGameParticipant _asfort_player_state_athena))))))
    (:then_1
      (return false))))

(fn SetEnableScene_Alert
  (params (bEnable :type Boolean))
  (bind _toyfoxdynamicuiscene_alert (Variables|Scenes|GetToyFoxDynamicUIScene_Alert))
  (bind _updated_variable (Utilities|FlowControl|DeltaGateWildcard (Variables|Transient|GetSceneEnable_Alert) bEnable)
    (:Success
      (if _updated_variable
        (DynamicUIDirector|AddScene _toyfoxdynamicuiscene_alert)
        (else
          (DynamicUIDirector|RemoveScene _toyfoxdynamicuiscene_alert))))))

(fn SetEnableScene_CommonHUD
  (params (bEnable :type Boolean))
  (bind _toyfoxdynamicuiscene_commonplayer_hud (Variables|Scenes|GetToyFoxDynamicUIScene_CommonPlayer_HUD))
  (bind _teammatewidgetscene (Variables|Default|GetTeammateWidgetScene))
  (bind _updated_variable (Utilities|FlowControl|DeltaGateWildcard (Variables|Transient|GetSceneEnable_CommonHUD) bEnable)
    (:Success
      (if _updated_variable
        (DynamicUIDirector|AddScene _toyfoxdynamicuiscene_commonplayer_hud)
        (if (not (DisplayHelper|IsUsingCompactUI))
          (DynamicUIDirector|AddScene _teammatewidgetscene))
        (else
          (DynamicUIDirector|RemoveScene _toyfoxdynamicuiscene_commonplayer_hud)
          (if (not (DisplayHelper|IsUsingCompactUI))
            (DynamicUIDirector|RemoveScene _teammatewidgetscene)))))))

(fn SetEnableScene_EnemyPlayers
  (params (IsEnabled :type Boolean))
  (bind _toyfoxdynamicuiscene_enemyplayers (Variables|Scenes|GetToyFoxDynamicUIScene_EnemyPlayers))
  (bind _toyfoxgamestatecomponent (Variables|Default|GetToyFoxGameStateComponent)
    (:then
      (Utilities|IsValid (Giant|GetGiantPossessingPlayerState _toyfoxgamestatecomponent)
        (:Is%20Valid
          (bind _updated_variable (Utilities|FlowControl|DeltaGateWildcard (Variables|Transient|GetSceneEnable_EnemyPlayers) (and (Variables|GetIsEnabled) (not (Team|OnSameTeam (Variables|/Script/Engine.Controller|GetPlayerState (DynamicUIDirector|GetOwningLocalPlayerController)) (Giant|GetGiantPossessingPlayerState _toyfoxgamestatecomponent)))))
            (:Success
              (if _updated_variable
                (DynamicUIDirector|AddScene _toyfoxdynamicuiscene_enemyplayers)
                (else
                  _)))))
        (:Is%20Not%20Valid
          (Utilities|FlowControl|DeltaGateWildcard (Variables|Transient|GetSceneEnable_EnemyPlayers) :!label _label
            (:Success
              _))))
      (DynamicUIDirector|RemoveScene _toyfoxdynamicuiscene_enemyplayers))
    (:else
      _))
  (goto _label))

(fn SetEnableScene_PrimaryPlayerHUD
  (params (bEnable :type Boolean))
  (bind _toyfoxdynamicuiscene_primaryplayer_hud (Variables|Scenes|GetToyFoxDynamicUIScene_PrimaryPlayer_HUD))
  (bind _updated_variable (Utilities|FlowControl|DeltaGateWildcard (Variables|Transient|GetSceneEnable_PrimaryPlayerHUD) bEnable)
    (:Success
      (if _updated_variable
        (DynamicUIDirector|AddScene _toyfoxdynamicuiscene_primaryplayer_hud)
        (else
          (DynamicUIDirector|RemoveScene _toyfoxdynamicuiscene_primaryplayer_hud))))))

(fn SetEnableScene_SecondaryPlayerHUD
  (params (bEnable :type Boolean))
  (bind _toyfoxdynamicuiscene_secondaryplayer_hud (Variables|Scenes|GetToyFoxDynamicUIScene_SecondaryPlayer_HUD))
  (bind _updated_variable (Utilities|FlowControl|DeltaGateWildcard (Variables|Transient|GetSceneEnable_SecondaryPlayerHUD) bEnable)
    (:Success
      (if _updated_variable
        (DynamicUIDirector|AddScene _toyfoxdynamicuiscene_secondaryplayer_hud)
        (CallFunction|SetScreenSizePercentageforRegionDivision 0.0)
        (else
          (DynamicUIDirector|RemoveScene _toyfoxdynamicuiscene_secondaryplayer_hud)
          (CallFunction|SetScreenSizePercentageforRegionDivision 0.5))))))

(fn SetHideAllHUDElements
  (params (bHide :type Boolean))
  (CallFunction|SetHideHUDElements bHide (Variables|Settings|GetAllHUDElementsToHide) "ToyFoxHideAllHUDElements"))

(fn SetHideHUDElements
  (params (bHideElements :type Boolean) (HUDElementsToHide :type Gameplay%20Tag%20Container%20Structure) (ReasonName :type Name :default "ToyFoxPilot"))
  (bind _asfort_player_controller (Utilities|Casting|CastToFortPlayerController (DynamicUIDirector|GetOwningLocalPlayerController)))
  (HUD|SetHUDElementVisibilitywithContext _asfort_player_controller HUDElementsToHide bHideElements ReasonName (Utilities|String|ToString(Name) ReasonName)))

(fn SetScreenSizePercentageForRegionDivision
  (params (ScreenSizePercentage :type Float%20%28double-precision%29))
  (bind _asfort_player_controller (Utilities|Casting|CastToFortPlayerController (DynamicUIDirector|GetOwningLocalPlayerController)))
  (if (or (TouchPlatform|IsUsingTouch _asfort_player_controller) (TouchPlatform|IsUsingMouseforTouch))
    (TouchControls|SetScreenSizePercentageforRegionDivision _asfort_player_controller ScreenSizePercentage)))

(fn ConstructionScript)
