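# Translucent UI material for ToyFox, named as a bottom blur, that draws a soft-edged, mirrored dark shape procedurally, without sampling textures or the scene.
# UVs get an optional Distortion warp and Padding scale, then pass through 2x1 symmetry into a rotated box SDF cut by a linear field.
# Smoothed SDF masks layer dark greys, a red band and a faded light outline from a separate SDF union.
# The shape and half-strength outline combine into an alpha mask.
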Material /ToyFoxGameplay/UI/Materials/M_ToyFox_BottomBlur
graph {
 properties {bAutomaticallySetUsageInEditor=false, blendMode=BLEND_Translucent, materialDomain=MD_UI, usedShadingModels=Default Lit}
 nodes [
  n1 = Constant3Vector() {constant=<rgba 0.0222 0.0222 0.0222 1>}
  n2 = Constant3Vector() {constant=<rgba 0.0137 0.0137 0.0137 1>}
  n3 = ScalarParameter() {defaultValue=-0.1, parameterName=Padding}
  n4 = MaterialFunctionCall(Float2=MaterialFunctionCall() {desc=Uvs, materialFunction=&GetUserInterfaceUV}.Pixel Size) {materialFunction=&BreakOutFloat2Components}
  n5 = AppendVector(A=Multiply(A=@n3, B=Divide(A=@n4.G, B=@n4.G)), B=@n3)
  n6 = ScalarParameter() {defaultValue=0.2, parameterName=Distortion}
  n7 = MaterialFunctionCall() {materialFunction=&GetUserInterfaceUV}
  n8 = MaterialFunctionCall(Float2=@n7.Normalized UV) {materialFunction=&BreakOutFloat2Components}
  n9 = MaterialFunctionCall(Float2=@n7.Normalized UV) {materialFunction=&BreakOutFloat2Components}
  n10 = Subtract(A=@n9.R)
  n11 = Multiply(A=@n6, B=Multiply(A=Multiply(A=@n10, B=@n9.R), B=Multiply(A=@n9.G) {constB=-1}))
  n12 = AppendVector(A=@n8.R, B=Saturate(None=Add(A=@n11, B=@n8.G)))
  n13 = MaterialFunctionCall!(
   UV=MaterialFunctionCall(UVs=If(A=@n6, "A > B"=@n12, "A == B"=@n7.Normalized UV, "A < B"=@n7.Normalized UV), Texture Scale=OneMinus(None=@n5)) {materialFunction=&ScaleUVsByCenter}.UVs
   Mirror Y
   Swizzle UV
  ) {materialFunction=/FortUILibrary/MaterialFunctions/Basics/MF_UI_Symmetry_2x1_New}
  n14 = Constant() {r=0.07}
  n15 = MaterialFunctionCall!(
   UVs=Add(
    A=MaterialFunctionCall(UVs=@n13.Symmetry, "Rotation Angle (0-1)"=@n14) {materialFunction=&CustomRotator}.Rotated Values
    B=Constant2Vector() {g=-0.6, r=0.2}
   )
   Size=Constant() {r=0.7}
   CornerRoundness
   KeepAspectRatio
   PixelSize
  ) {materialFunction=/FortUILibrary/MaterialFunctions/Shapes/SDFs/MF_SDF_Box}
  n16 = MaterialFunctionCall(Float2=@n13.Symmetry) {materialFunction=&BreakOutFloat2Components}
  n17 = MaterialFunctionCall!(DistanceFieldA=@n15.SDF, DistanceFieldB=Multiply(A=Add(A=@n16.R) {constB=-0.825}) {constB=-2}, BlendAmount) {materialFunction=/FortUILibrary/MaterialFunctions/Shapes/Modifiers/MF_UI_SDF_Subtraction}
  n18 = Constant() {r=0.1}
  n19 = MaterialFunctionCall(DistanceField=@n17.Exact, Threshold=Constant() {r=-0.2}, Smoothing=@n18) {materialFunction=&MF_UI_AntiAliasing}
  n20 = MaterialFunctionCall(DistanceField=@n17.Exact, Smoothing=@n18) {materialFunction=&MF_UI_AntiAliasing}
  n21 = Multiply(A=Power(Base=@n16.R) {constExponent=0.5}) {constB=0.8}
  n22 = Subtract(
   A=@n20.Inner
   B=MaterialFunctionCall(DistanceField=Add(A=@n17.Exact, B=@n21), Threshold=Constant() {r=0.675}, Smoothing=Constant() {r=0.2}) {materialFunction=&MF_UI_AntiAliasing}.Inner
  )
  n23 = Add(A=Multiply(A=OneMinus(None=@n16.G))) {constB=0}
  n24 = Multiply(
   A=@n16.R
   B=MaterialFunctionCall!(
    DistanceField=MaterialFunctionCall!(DistanceFieldA=@n15.SDF, DistanceFieldB=@n23, BlendAmount) {materialFunction=/FortUILibrary/MaterialFunctions/Shapes/Modifiers/MF_UI_SDF_Union}.Exact
    Threshold
    Thickness=Constant() {r=0.015}
    Alignment=Constant() {r=9}
    Smoothing
   ) {materialFunction=/FortUILibrary/MaterialFunctions/Shapes/Modifiers/MF_UI_ShapeOutline}.Outline
  )
  LinearInterpolate(
   A=LinearInterpolate(A=LinearInterpolate(A=@n1, B=@n2, Alpha=@n19.Inner), B=Constant3Vector() {constant=<rgba 0.3594 0 0.0076 1>}, Alpha=Saturate(None=@n22))
   Alpha=@n24
  )
  Max(A=@n20.Inner, B=Multiply(A=@n24) {constB=0.5})
  OneMinus(None=@n8.G)
 ]
}
