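# Translucent UI material for the ToyFox main player reticle, drawn procedurally with signed distance fields.
# An outlined circle SDF is negated and combined through SDF subtraction with a box evaluated in 2x2 mirrored UVs, shaping a symmetric ring-based reticle.
# The result is anti-aliased at two thresholds and blended at a fixed ratio.
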Material /ToyFoxGameplay/UI/Materials/M_ToyFox_MainPlayerReticle
graph {
 properties {bAutomaticallySetUsageInEditor=false, blendMode=BLEND_Translucent, materialDomain=MD_UI, usedShadingModels=Default Lit}
 nodes [
  n1 = MaterialFunctionCall!(UVs, Radius=Constant() {r=0.7}) {materialFunction=/FortUILibrary/MaterialFunctions/Shapes/SDFs/MF_SDF_Circle}
  n2 = Constant() {r=0.045}
  n3 = MaterialFunctionCall!(
   DistanceFieldA=MaterialFunctionCall!(
    UVs=MaterialFunctionCall!(UV, Swizzle UV) {materialFunction=/FortUILibrary/MaterialFunctions/Basics/MF_UI_Symmetry_2x2_New}.Symmetry
    Size=Constant() {r=0.7}
    CornerRoundness
    KeepAspectRatio
    PixelSize
   ) {materialFunction=/FortUILibrary/MaterialFunctions/Shapes/SDFs/MF_SDF_Box}.SDF
   DistanceFieldB=Multiply(A=MaterialFunctionCall!(DistanceField=@n1.SDF, Threshold, Thickness=@n2, Alignment, Smoothing) {materialFunction=/FortUILibrary/MaterialFunctions/Shapes/Modifiers/MF_UI_ShapeOutline}.SDFOutline) {constB=-1}
   BlendAmount
  ) {materialFunction=/FortUILibrary/MaterialFunctions/Shapes/Modifiers/MF_UI_SDF_Subtraction}
  LinearInterpolate(
   A=MaterialFunctionCall(DistanceField=@n3.Exact) {materialFunction=&MF_UI_AntiAliasing}.Inner
   B=MaterialFunctionCall(DistanceField=@n3.Exact, Threshold=Constant() {r=0.03}) {materialFunction=&MF_UI_AntiAliasing}.Inner
  ) {constAlpha=0.4}
 ]
}
