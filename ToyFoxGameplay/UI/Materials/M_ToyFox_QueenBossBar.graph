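# Translucent UI material for the ToyFox Queen boss bar, building a procedural shape mask from signed distance fields with no textures.
# A wide trapezoid, an angled edge from mirrored, rotated UVs, and an offset second trapezoid combine through two chained SDF subtractions.
# The resulting field yields a thin outline plus a half-intensity antialiased fill, summed in a final Add.
# The graph shows no color, parameters or link from that Add to a material output.
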Material /ToyFoxGameplay/UI/Materials/M_ToyFox_QueenBossBar
graph {
 properties {bAutomaticallySetUsageInEditor=false, blendMode=BLEND_Translucent, materialDomain=MD_UI, usedShadingModels=Default Lit}
 nodes [
  n1 = MaterialFunctionCall() {materialFunction=&GetUserInterfaceUV}
  n2 = Constant() {r=0.7}
  n3 = MaterialFunctionCall!(UVs=@n1.Normalized UV, TopWidth=@n2, BottomWidth=Add(A=@n2) {constB=0.1}, Height=Constant() {r=0.15}) {materialFunction=/FortUILibrary/MaterialFunctions/Shapes/SDFs/MF_SDF_IsoscelesTrapezoid}
  n4 = MaterialFunctionCall!(UV=@n1.Normalized UV, Mirror Y, Swizzle UV) {materialFunction=/FortUILibrary/MaterialFunctions/Basics/MF_UI_Symmetry_2x1_New}
  n5 = Constant() {r=0.04}
  n6 = ComponentMask(None=Add(A=MaterialFunctionCall(UVs=@n4.Symmetry, "Rotation Angle (0-1)"=@n5) {materialFunction=&CustomRotator}.Rotated Values) {constB=-0.25}) {r=true}
  n7 = Add(A=@n1.Normalized UV, B=Constant2Vector() {g=-0.2})
  n8 = Constant() {r=0.5}
  n9 = Add(A=@n8) {constB=0.3}
  n10 = Constant() {r=0.425}
  n11 = MaterialFunctionCall!(
   DistanceFieldA=MaterialFunctionCall(DistanceFieldA=Multiply(A=@n3.SDF) {constB=1.25}, DistanceFieldB=Multiply(A=@n6) {constB=2}) {materialFunction=/FortUILibrary/MaterialFunctions/Shapes/Modifiers/MF_UI_SDF_Subtraction}.Exact
   DistanceFieldB=Multiply(A=MaterialFunctionCall(UVs=@n7, TopWidth=@n8, BottomWidth=@n9, Height=@n10) {materialFunction=&MF_SDF_IsoscelesTrapezoid}.SDF) {constB=1.25}
   BlendAmount
  ) {materialFunction=&MF_UI_SDF_Subtraction}
  Add(
   A=MaterialFunctionCall!(DistanceField=@n11.Exact, Threshold, Thickness=Constant() {r=0.02}, Alignment, Smoothing) {materialFunction=/FortUILibrary/MaterialFunctions/Shapes/Modifiers/MF_UI_ShapeOutline}.Outline
   B=Multiply(A=MaterialFunctionCall(DistanceField=@n11.Exact) {materialFunction=&MF_UI_AntiAliasing}.Inner) {constB=0.5}
  )
 ]
}
