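# UI material for a team viewer ability icon, drawing a red-tinted, skewed backer behind a swappable icon texture.
# ReadySwitch blends the backer colors, icon brightness and center-based icon scale.
# Skewed box SDFs with antialiasing build the backer, border and highlight masks, which combine with icon alpha for color and translucency.
# HighlightSwitch fades in a vertically graded red highlight.
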Material /ToyFoxGameplay/UI/Materials/M_ToyFox_TeamViewer_AbilityIconBacker
graph {
 properties {bAutomaticallySetUsageInEditor=false, blendMode=BLEND_Translucent, materialDomain=MD_UI, usedShadingModels=Default Lit}
 nodes [
  n1 = ScalarParameter() {parameterName=ReadySwitch}
  n2 = LinearInterpolate(A=Constant3Vector() {constant=<rgba 0.0648 0.0144 0.0144 1>}, B=Constant3Vector() {constant=<rgba 0.063 0.0252 0.0252 1>}, Alpha=@n1)
  n3 = NamedRerouteUsage()
  n4 = LinearInterpolate(A=Constant3Vector() {constant=<rgba 0.0538 0.027 0.027 1>}, B=Constant3Vector() {constant=<rgba 0.1753 0.0277 0.0277 1>}, Alpha=@n1)
  n5 = MaterialFunctionCall() {materialFunction=&GetUserInterfaceUV}
  n6 = Constant2Vector() {g=-0.15}
  n7 = ScalarParameter() {parameterName=ReadySwitch}
  n8 = TextureSampleParameter2D(UVs=MaterialFunctionCall(UVs=Add(A=@n5.Normalized UV, B=@n6), Texture Scale=LinearInterpolate(Alpha=@n7) {constA=0.5, constB=0.7}) {materialFunction=&ScaleUVsByCenter}.UVs) {
   parameterName=TeamAbilityIcon
   samplerSource=SSM_Clamp_WorldGroupSettings
   texture=/Game/Athena/UI/InGame/Creative/PropertyEditor/Icon_Picker/Textures/64x64/T_UI_IconLibrary_sword_64
  }
  n9 = NamedRerouteDeclaration(None=@n8.A) {name=IconFill, nodeColor=<rgba 0.75 0.6098 0.375 0.75>}
  Add(
   A=LinearInterpolate(A=LinearInterpolate(A=@n2, B=@n4, Alpha=@n3), B=LinearInterpolate(Alpha=@n1) {constA=0.25, constB=0.8}, Alpha=@n9) {constB=0.8}
   B=LinearInterpolate(B=Constant3Vector() {constant=<rgba 0.5191 0.0496 0.0643 1>}, Alpha=NamedRerouteUsage())
  )
  n10 = MaterialFunctionCall!(UV=Add(A=@n5.Normalized UV, B=Constant2Vector() {g=-0.15}), Skew Angle=Constant2Vector() {r=10}) {materialFunction=/FortUILibrary/MaterialFunctions/Basics/MF_UI_Skew}
  n11 = Constant2Vector() {g=0.5, r=0.625}
  n12 = MaterialFunctionCall!(UVs=@n10.Skewed UVs, Size=@n11, CornerRoundness, KeepAspectRatio, PixelSize) {materialFunction=/FortUILibrary/MaterialFunctions/Shapes/SDFs/MF_SDF_Box}
  n13 = MaterialFunctionCall(DistanceField=@n12.SDF) {materialFunction=&MF_UI_AntiAliasing}
  n14 = Constant() {r=-0.05}
  n15 = MaterialFunctionCall(DistanceField=@n12.SDF, Threshold=@n14) {materialFunction=&MF_UI_AntiAliasing}
  n16 = NamedRerouteDeclaration(None=LinearInterpolate(A=@n13.Inner, B=@n15.Inner) {constAlpha=0.1}) {name=BackerMask, nodeColor=<rgba 0.75 0.6098 0.375 0.75>}
  n17 = Constant2Vector() {g=0.225}
  n18 = Constant2Vector() {g=0.4, r=-0.05}
  n19 = MaterialFunctionCall(
   DistanceField=MaterialFunctionCall(UVs=Add(A=@n10.Skewed UVs, B=@n17), Size=Add(A=@n11, B=@n18)) {materialFunction=&MF_SDF_Box}.SDF
   Threshold=@n14
  ) {materialFunction=&MF_UI_AntiAliasing}
  n20 = Power(Base=Add(A=ComponentMask(None=@n5.Normalized UV) {g=true}) {constB=-0.1}) {constExponent=0.4}
  n21 = Multiply(A=LinearInterpolate(B=Multiply(A=@n19.Inner, B=@n20), Alpha=ScalarParameter() {parameterName=HighlightSwitch}))
  Multiply(A=Add(A=LinearInterpolate(A=@n16, Alpha=@n9), B=@n21))
  NamedRerouteDeclaration(None=Subtract(A=@n13.Inner, B=@n15.Inner)) {name=BorderMask, nodeColor=<rgba 0.75 0.6098 0.375 0.75>}
  NamedRerouteDeclaration(None=@n21) {name=HighlightMask, nodeColor=<rgba 0.75 0.6098 0.375 0.75>}
 ]
}
