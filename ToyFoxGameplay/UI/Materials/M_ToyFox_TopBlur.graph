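# Translucent UI material for ToyFox that builds a soft-edged procedural mask, named as a top blur, without sampling any textures.
# A positive Distortion parameter bends the UI's vertical coordinate along a parabola across the width, scaled by height.
# The bent UVs, offset and scaled, feed two nine-slice isosceles trapezoid SDFs.
# Each SDF is negated and smoothstepped, then the second is subtracted from the first to carve the final shape.
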Material /ToyFoxGameplay/UI/Materials/M_ToyFox_TopBlur
graph {
 properties {bAutomaticallySetUsageInEditor=false, blendMode=BLEND_Translucent, materialDomain=MD_UI, usedShadingModels=Default Lit}
 nodes [
  Constant3Vector() {constant=<rgba 0.0168 0.011 0.0137 1>}
  n1 = Constant2Vector() {g=0.35}
  n2 = ScalarParameter() {defaultValue=1, parameterName=Distortion}
  n3 = MaterialFunctionCall() {materialFunction=&GetUserInterfaceUV}
  n4 = MaterialFunctionCall(Float2=@n3.Normalized UV) {materialFunction=&BreakOutFloat2Components}
  n5 = MaterialFunctionCall(Float2=@n3.Normalized UV) {materialFunction=&BreakOutFloat2Components}
  n6 = Multiply(A=Subtract(A=@n5.R), B=@n5.R)
  n7 = Multiply(A=@n5.G) {constB=2}
  n8 = Add(A=Multiply(A=@n2, B=Multiply(A=@n6, B=@n7)), B=@n4.G)
  n9 = If(A=@n2, "A > B"=AppendVector(A=@n4.R, B=Saturate(None=@n8)), "A == B"=@n3.Normalized UV, "A < B"=@n3.Normalized UV)
  n10 = Add(A=@n1, B=@n9)
  n11 = Multiply(A=MaterialFunctionCall!(
    UVs=MaterialFunctionCall!(UVs=@n10) {materialFunction=/FortUILibrary/MaterialFunctions/Basics/MF_UI_9SliceUVs}.UVs
    TopWidth=Constant() {r=1}
    BottomWidth=Constant()
    Height=Constant() {r=0.75}
   ) {materialFunction=/FortUILibrary/MaterialFunctions/Shapes/SDFs/MF_SDF_IsoscelesTrapezoid}.SDF) {constB=-1}
  n12 = Constant2Vector() {g=-0.3}
  n13 = MaterialFunctionCall(UVs=Add(A=@n9, B=Add(A=@n1, B=@n12)), Texture Scale=Constant2Vector() {g=1, r=0.5}) {materialFunction=&ScaleUVsByCenter}
  n14 = Multiply(A=MaterialFunctionCall(
    UVs=MaterialFunctionCall(UVs=@n13.UVs) {materialFunction=&MF_UI_9SliceUVs}.UVs
    TopWidth=Constant()
    BottomWidth=Constant() {r=1.5}
    Height=Constant() {r=0.5}
   ) {materialFunction=&MF_SDF_IsoscelesTrapezoid}.SDF) {constB=-1}
  Subtract(A=Saturate(None=SmoothStep(Value=@n11) {constMax=0.2}), B=Saturate(None=SmoothStep(Value=@n14) {constMax=0.2}))
 ]
}
