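# UI material for a ToyFox team viewer glow, an alpha-composited lightning effect in red tones scaled by a Master Opacity parameter.
# Two lightning texture channels are thresholded into cores plus soft glows, masked, and crossfaded on DebugTimeSine.
# A stepped sine alternates the red tints between layers.
# Rotated hexagon SDFs, an edge mask and two sine-swept shine bands supply procedural masks through named reroutes.
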Material /ToyFoxGameplay/UI/Materials/M_UI_ToyFox_TeamViewer_Glow
graph {
 properties {bAutomaticallySetUsageInEditor=false, blendMode=BLEND_AlphaComposite, materialDomain=MD_UI, usedShadingModels=Default Lit}
 nodes [
  n1 = NamedRerouteUsage()
  Add(A=NamedRerouteUsage(), B=@n1)
  Constant3Vector() {constant=<rgba 1 0 0.1372 1>}
  Constant() {r=1}
  Constant() {r=1}
  MaterialFunctionCall() {materialFunction=&DebugTimeSine}
  MaterialFunctionCall() {materialFunction=&CustomRotator}
  n2 = Time()
  MaterialFunctionCall("Value (0-1)"=Frac(None=Multiply(A=@n2))) {materialFunction=&MF_UI_EaseCurves}
  MaterialFunctionCall() {materialFunction=&DebugTimeSine}
  n3 = Step(Y=MaterialFunctionCall(Speed=Constant() {r=1}) {materialFunction=&DebugTimeSine}.Result) {constX=0.5}
  n4 = Multiply(A=LinearInterpolate(B=Constant3Vector() {constant=<rgba 1 0 0 0>}, Alpha=@n3) {constA=0.1}, B=NamedRerouteUsage())
  n5 = Multiply(A=LinearInterpolate(A=Constant3Vector() {constant=<rgba 0.4375 0 0.0016 1>}, Alpha=@n3) {constB=0.1}, B=NamedRerouteUsage())
  n6 = Multiply(A=Multiply(A=@n1, B=NamedRerouteUsage()))
  n7 = ScalarParameter() {defaultValue=1, parameterName=Master Opacity}
  Multiply(A=Multiply(A=Add(A=@n4, B=@n5), B=@n6), B=@n7)
  Multiply(A=@n6, B=@n7)
  n8 = NamedRerouteUsage()
  n9 = AppendVector(A=LinearInterpolate(Alpha=Constant() {r=1}) {constB=0.7}, B=Constant())
  n10 = MaterialFunctionCall!(
   UVs=MaterialFunctionCall(UVs=Add(A=@n8, B=@n9), "Rotation Angle (0-1)"=Constant() {r=0.25}) {materialFunction=&CustomRotator}.Rotated Values
   Radius=Constant() {r=0.6}
   Sides=Constant() {r=6}
   InnerAngle
  ) {materialFunction=/FortUILibrary/MaterialFunctions/Shapes/SDFs/MF_SDF_RegularPolygon}
  n11 = Constant() {r=0.5}
  NamedRerouteDeclaration(None=Power(Base=MaterialFunctionCall(DistanceField=@n10.SDF, Smoothing=@n11) {materialFunction=&MF_UI_AntiAliasing}.Inner) {constExponent=4}) {name=Hex Glow, nodeColor=<rgba 0 0 0 1>}
  n12 = NamedRerouteUsage()
  n13 = NamedRerouteUsage()
  n14 = Constant()
  NamedRerouteDeclaration(None=LinearInterpolate(
    A=Saturate(None=@n12)
    B=Saturate(None=@n13)
    Alpha=MaterialFunctionCall(Speed=@n14) {materialFunction=&DebugTimeSine}.Result
   )) {name=LightningCombined, nodeColor=<rgba 0 0.9297 1 1>}
  n15 = TextureSample() {texture=/ToyFoxGameplay/UI/Materials/Textures/T_UI_ToyFox_LightningPattern}
  n16 = Constant() {r=0.4}
  n17 = Constant() {r=0.1}
  n18 = NamedRerouteUsage()
  n19 = Multiply(
   A=Multiply(A=MaterialFunctionCall(DistanceField=@n15.R, Threshold=@n16, Smoothing=@n17) {materialFunction=&MF_UI_AntiAliasing}.Outer) {constB=0.4}
   B=@n18
  )
  n20 = Multiply(
   A=MaterialFunctionCall(DistanceField=@n15.R, Threshold=Constant() {r=0.55}, Smoothing=Constant()) {materialFunction=&MF_UI_AntiAliasing}.Outer
   B=@n18
  )
  NamedRerouteDeclaration(None=Add(
    A=NamedRerouteDeclaration(None=@n19) {name=L01 Glow, nodeColor=<rgba 1 0 0.038 1>}
    B=NamedRerouteDeclaration(None=@n20) {name=L01, nodeColor=<rgba 1 0 0.038 1>}
   )) {name=L01 Combined, nodeColor=<rgba 1 0 0.038 1>}
  n21 = TextureSample() {texture=&T_UI_ToyFox_LightningPattern}
  n22 = Constant() {r=0.5}
  n23 = Constant()
  n24 = NamedRerouteUsage()
  n25 = Multiply(
   A=Multiply(A=MaterialFunctionCall(DistanceField=@n21.G, Threshold=@n22, Smoothing=@n23) {materialFunction=&MF_UI_AntiAliasing}.Outer)
   B=@n24
  )
  n26 = Constant() {r=0.5}
  n27 = Constant() {r=0.2}
  n28 = Multiply(
   A=Multiply(A=MaterialFunctionCall(DistanceField=@n21.G, Threshold=@n26, Smoothing=@n27) {materialFunction=&MF_UI_AntiAliasing}.Outer) {constB=0.6}
   B=@n24
  )
  NamedRerouteDeclaration(None=Add(
    A=NamedRerouteDeclaration(None=@n25) {name=L02, nodeColor=<rgba 0 0.9016 1 1>}
    B=NamedRerouteDeclaration(None=@n28) {name=L02 Glow, nodeColor=<rgba 0 0.9016 1 1>}
   )) {name=L02 Combined, nodeColor=<rgba 1 0.9375 0 1>}
  n29 = ComponentMask(None=TextureCoordinate()) {r=true}
  n30 = OneMinus(None=@n29)
  NamedRerouteDeclaration(None=Multiply(A=Multiply(A=@n30, B=@n29)) {constB=3}) {name=EdgeMask, nodeColor=<rgba 0 0.2734 1 1>}
  n31 = Constant() {r=0.5}
  n32 = Add(
   A=ComponentMask(None=TextureCoordinate()) {r=true}
   B=LinearInterpolate(
    A=Constant() {r=-1}
    B=Constant() {r=1}
    Alpha=MaterialFunctionCall(Speed=@n31) {materialFunction=&DebugTimeSine}.Result
   )
  )
  n33 = OneMinus(None=@n32)
  NamedRerouteDeclaration(None=Multiply(A=Multiply(A=@n33, B=@n32)) {constB=3}) {name=Shine, nodeColor=<rgba 0.125 1 0 1>}
  n34 = Constant() {r=0.1}
  n35 = Add(
   A=ComponentMask(None=TextureCoordinate()) {r=true}
   B=LinearInterpolate(
    A=Constant() {r=-1}
    B=Constant() {r=1}
    Alpha=MaterialFunctionCall(Speed=@n34) {materialFunction=&DebugTimeSine}.Result
   )
  )
  n36 = OneMinus(None=@n35)
  NamedRerouteDeclaration(None=Multiply(A=Multiply(A=@n36, B=@n35)) {constB=3}) {name=Shine 02, nodeColor=<rgba 1 0 0.6094 1>}
  NamedRerouteDeclaration(None=MaterialFunctionCall() {materialFunction=&GetUserInterfaceUV}.Normalized UV) {name=Normalized UV, nodeColor=<rgba 0 0 0 1>}
  NamedRerouteDeclaration(None=MaterialFunctionCall() {materialFunction=&MF_UI_UVScaling}.Scaled to Fit) {name=Scale to Fit, nodeColor=<rgba 0 0 0 1>}
  n37 = NamedRerouteUsage()
  n38 = AppendVector(A=LinearInterpolate(Alpha=ScalarParameter() {defaultValue=1, parameterName=X}) {constB=0.7}, B=Constant())
  n39 = MaterialFunctionCall(
   UVs=MaterialFunctionCall(UVs=Add(A=@n37, B=@n38), "Rotation Angle (0-1)"=Constant() {r=0.25}) {materialFunction=&CustomRotator}.Rotated Values
   Radius=Constant() {r=0.5}
   Sides=Constant() {r=6}
  ) {materialFunction=&MF_SDF_RegularPolygon}
  n40 = Constant()
  NamedRerouteDeclaration(None=Power(Base=MaterialFunctionCall(DistanceField=@n39.SDF, Smoothing=@n40) {materialFunction=&MF_UI_AntiAliasing}.Inner) {constExponent=4}) {name=Hex Lag, nodeColor=<rgba 0 0 0 1>}
  NamedRerouteUsage()
  NamedRerouteUsage()
  Panner(Speed=Constant2Vector() {r=-1})
  Panner(Speed=Constant2Vector() {r=-0.7})
  Panner()
  Step(Y=MaterialFunctionCall(Speed=Constant() {r=1}) {materialFunction=&DebugTimeSine}.Result) {constX=0.5}
 ]
}
