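# UI material function that converts a local U or V value so it is relative to the widget's live pixel size on one axis.
# It divides the authored pixel size by the live widget dimension, multiplies by the viewport scale factor from MF_UI_ViewportScale, then scales the input value by the result.
# The caller supplies the live dimension, typically a TexCoord[3] component; the TexCoord[3] node inside the function is unconnected.
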MaterialFunction /ToyFoxGameplay/UI/Materials/Functions/MF_UI_ToyFox_PixelRatio
graph {nodes [
  n1 = FunctionInput() {
   description="The dim of the widget the material is applied to, in pixels, in the same axis at the Local U/V Valu"
   inputName=Authored Widget Px Value
   inputType=FunctionInput_Scalar
   sortPriority=10
  }
  n2 = FunctionInput() {
   description="The live single-axis dims of the widget. e.g. TexCoord[3].x"
   inputName=Widget Axis Px Dims
   inputType=FunctionInput_Scalar
   sortPriority=20
  }
  n3 = Divide(A=@n1, B=@n2) {desc=Ratio}
  n4 = MaterialFunctionCall() {materialFunction=/FortUILibrary/MaterialFunctions/Basics/MF_UI_ViewportScale}
  FunctionOutput(None=Multiply(
    A=FunctionInput() {description=e.g. a Width/X value, inputName=Local U/V-RelativeValue, inputType=FunctionInput_Scalar}
    B=Multiply(A=@n3, B=@n4.Viewport Scale Factor)
   ))
  TextureCoordinate() {coordinateIndex=3, desc=WidgetPx}
 ]}
