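# UI material for a ToyFox health top bar, drawing a faint grey, alpha-composited border shape with a Flash parameter.
# Normalized UI UVs are nudged vertically, passed through a zero-angle rotator, center-scaled and fill-scaled to sample an MSDF border texture.
# The MSDF becomes an SDF, and an antialiased threshold yields the shape mask.
# That mask scales the Flash-blended color and, separately, a low alpha value.
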Material /ToyFoxGameplay/UI/Materials/M_ToyFox_HealthTopBar
graph {
 properties {bAutomaticallySetUsageInEditor=false, blendMode=BLEND_AlphaComposite, materialDomain=MD_UI, usedShadingModels=Default Lit}
 nodes [
  Constant3Vector() {constant=<rgba 1 0 0 0>}
  MaterialFunctionCall() {materialFunction=&MF_UI_UVScaling}
  n1 = MaterialFunctionCall() {materialFunction=&GetUserInterfaceUV}
  n2 = Add(
   A=AppendVector(A=ComponentMask(None=@n1.Normalized UV) {r=true}, B=ComponentMask(None=@n1.Normalized UV) {g=true})
   B=Constant2Vector() {g=-0.05}
  )
  n3 = Constant()
  n4 = MaterialFunctionCall(UV=MaterialFunctionCall(
    UVs=MaterialFunctionCall(UVs=@n2, "Rotation Angle (0-1)"=@n3) {materialFunction=&CustomRotator}.Rotated Values
    Texture Scale=Constant() {r=0.9}
   ) {materialFunction=&ScaleUVsByCenter}.UVs) {materialFunction=&MF_UI_UVScaling}
  n5 = MaterialFunctionCall(
   DistanceField=MaterialFunctionCall!(MSDF RGB=TextureSample(UVs=@n4.Scaled To Fill) {samplerType=SAMPLERTYPE_LinearColor, texture=/ToyFoxGameplay/UI/Materials/Textures/T_UI_ToyFox_borderBottom}.RGB) {materialFunction=/FortUILibrary/MaterialFunctions/Shapes/Modifiers/MF_UI_MSDFtoSDF}.SDF
   Threshold=Constant() {r=0.65}
  ) {materialFunction=&MF_UI_AntiAliasing}
  Multiply(
   A=LinearInterpolate(A=Constant3Vector() {constant=<rgba 0.2865 0.273 0.2732 1>}, Alpha=ScalarParameter() {parameterName=Flash})
   B=Multiply(A=@n5.Outer) {constB=0.3}
  )
  Multiply(A=@n5.Outer, B=Constant() {r=0.1}) {constB=0.5}
 ]
}
